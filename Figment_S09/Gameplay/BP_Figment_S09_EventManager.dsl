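; Figment S09 event manager that drives the Cat Eye world event from calendar events.
; On the server, after match ready and calendar sync, it maps Cat1 (breathing, eye visible) or Cat0 (breathing) to a state.
; It broadcasts that state through the global GameplayEventRouter as a stateful event, then listens to it to swap active and inactive CatEye Data Layers.
; Test events force each state, including inactive.
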(event EventBeginPlay
  (Networking|SwitchHasAuthority
    (:Authority
      (Class|FortAsyncActionAuthorityMatchReady|MatchReadyAsync
        (:OnReady
          (MCPContext|InitialCalendarSyncCompleted
            (:OnSyncComplete
              (CallFunction|SetupCatEyeState)
              (CallFunction|SetupDataLayers))))))))

(event Custom|TestSetCatActive
  ; Cat1 Event - Cat Breathing and Eye visible
  (Variables|Default|SetCatEyeEventState (Utilities|Struct|MakeSTFigmentCatEyeEvent "NewEnumerator2"))
  (CallFunction|NotifyCatEyeStateChanged))

(event Custom|SetupCatEyeState
  (bind _biseventactive (CalendarEvents|GetCalendarEventInformation "Cat1"))
  (if _biseventactive
    ; Cat1 Event - Cat Breathing and Eye visible
    (Variables|Default|SetCatEyeEventState (Utilities|Struct|MakeSTFigmentCatEyeEvent "NewEnumerator2"))
    (CallFunction|NotifyCatEyeStateChanged)
    (else
      (bind _biseventactive_2 (CalendarEvents|GetCalendarEventInformation "Cat0"))
      (if _biseventactive_2
        ; Cat0 Event - Cat Breathing
        (Variables|Default|SetCatEyeEventState (Utilities|Struct|MakeSTFigmentCatEyeEvent "NewEnumerator1"))
        (CallFunction|NotifyCatEyeStateChanged)))))

(event Custom|TestSetCatBreathing
  ; Cat0 Event - Cat Breathing
  (Variables|Default|SetCatEyeEventState (Utilities|Struct|MakeSTFigmentCatEyeEvent "NewEnumerator1"))
  (CallFunction|NotifyCatEyeStateChanged))

(event Custom|NotifyCatEyeStateChanged
  (bind _returnvalue (GameplayEventRouter|BroadcastStatefulEvent (GameplayEventRouter|GetGlobalEventRouter) (Variables|Default|GetCatEyeEventState) 0))
  (if (not _returnvalue)
    (Development|PrintString "Warning: Changed CatEye Event, but noghing listened to it" true true "(R=1.000000,G=0.000000,B=0.000000,A=1.000000)")))

(event Custom|SetupDataLayers
  (bind (_asynctaskproxy _payload) (GameplayEventRouter|StartListeningToStatefulEvent (GameplayEventRouter|GetGlobalEventRouter) "/Figment_S09/Gameplay/EventRouter/ST_Figment_CatEye_Event.ST_Figment_CatEye_Event")
    (:OnEventReceived
      _)
    (:OnSavedEventStateExists
      _))
  (switch Utilities|FlowControl|Switch|SwitchonEnum_CatEye_Event_State (Utilities|Struct|BreakSTFigmentCatEyeEvent _payload)
    (:NewEnumerator0
      ; Deactivate CatEye Data layers. This is for testing only, shouldn't happen during the game
      (bind _returnvalue (WorldPartition|GetDataLayerManager))
      (DataLayers|SetDataLayerRuntimeState _returnvalue (Variables|Default|GetInactiveCatEyeDataLayer) "Activated" false :!class /Script/Engine.DataLayerManager)
      (DataLayers|SetDataLayerRuntimeState _returnvalue (Variables|Default|GetCatEyeDataLayer) "Unloaded" false :!class /Script/Engine.DataLayerManager))
    (:NewEnumerator1
      _)
    (:NewEnumerator2
      _))
  ; Activate CatEye Datalayer
  (bind _returnvalue_2 (WorldPartition|GetDataLayerManager))
  (DataLayers|SetDataLayerRuntimeState _returnvalue_2 (Variables|Default|GetInactiveCatEyeDataLayer) "Unloaded" false :!class /Script/Engine.DataLayerManager)
  (DataLayers|SetDataLayerRuntimeState _returnvalue_2 (Variables|Default|GetCatEyeDataLayer) "Activated" false :!class /Script/Engine.DataLayerManager))

(event Custom|TestSetCatInactive
  ; TestInactivateCat
  (Variables|Default|SetCatEyeEventState (Utilities|Struct|MakeSTFigmentCatEyeEvent "NewEnumerator0"))
  (CallFunction|NotifyCatEyeStateChanged))

(fn ConstructionScript)
